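# Physics asset for the male medium Tiptoe player cosmetic, covering the dynamic hood bones around the head.
# Capsule bodies form top, left and right chains, each running from the head through a tiny base body to a larger hood body.
# Base bodies rotate freely on the head, while the hood bodies have tight angular limits relative to their bases.
# The back hood body attaches directly to the head with narrow swing and twist limits.
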PhysicsAsset /BRCosmetics/Characters/Player/Male/Medium/Bodies/M_MED_Tiptoe/Meshes/Parts/M_MED_TipToe_Physics
bodies {
 dyn_hood_back [{shapeType=Capsule, length=2.75, radius=3.25, center=<4 0 0>, rotation=<rot 90 0 0>}]
 dyn_hood_base_l [{shapeType=Capsule, length=0.505, radius=0.505, rotation=<rot 0 -90 0>}]
 dyn_hood_base_r [{shapeType=Capsule, length=0.505, radius=0.505, rotation=<rot 0 -90 0>}]
 dyn_hood_l [{shapeType=Capsule, length=2.5, radius=3.25, center=<5 0 0>, rotation=<rot 90 0 0>}]
 dyn_hood_r [{shapeType=Capsule, length=2.5, radius=3.25, center=<5 0 0>, rotation=<rot 90 0 0>}]
 dyn_hood_top [{shapeType=Capsule, length=2.5, radius=3.25, center=<3.9104 0 0>, rotation=<rot 90 0 0>}]
 dyn_hood_top_base [{shapeType=Capsule, length=0.505, radius=0.505, rotation=<rot 0 -90 0>}]
 head [{shapeType=Capsule, length=1, radius=1}]
}
constraints [
 {bone1Name=dyn_hood_top, bone2Name=dyn_hood_top_base, swing1LimitDegrees=5, swing2LimitDegrees=5, twistLimitDegrees=10}
 {bone1Name=dyn_hood_l, bone2Name=dyn_hood_base_l, swing1LimitDegrees=5, swing2LimitDegrees=15, twistLimitDegrees=25}
 {bone1Name=dyn_hood_r, bone2Name=dyn_hood_base_r, swing1LimitDegrees=5, swing2LimitDegrees=15, twistLimitDegrees=25}
 {
  bone1Name=dyn_hood_top_base
  bone2Name=head
  swing1LimitDegrees=45
  swing1Motion=Free
  swing2LimitDegrees=45
  swing2Motion=Free
  twistLimitDegrees=45
  twistMotion=Free
 }
 {
  bone1Name=dyn_hood_base_l
  bone2Name=head
  swing1LimitDegrees=45
  swing1Motion=Free
  swing2LimitDegrees=45
  swing2Motion=Free
  twistLimitDegrees=45
  twistMotion=Free
 }
 {
  bone1Name=dyn_hood_base_r
  bone2Name=head
  swing1LimitDegrees=45
  swing1Motion=Free
  swing2LimitDegrees=45
  swing2Motion=Free
  twistLimitDegrees=45
  twistMotion=Free
 }
 {bone1Name=dyn_hood_back, bone2Name=head, swing1LimitDegrees=10, swing2LimitDegrees=10, twistLimitDegrees=15}
]
