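# Animation Blueprint for the TipToe male body cosmetic, posing the mesh and animating its hood.
# It copies the attached parent's pose, layers upper-arm branches copied from Body_Mesh (its owning component), then applies additive right-hood bone corrections and RigidBody simulation.
# Right upper-arm pitch and yaw relative to the clavicle drive the correction, which is disabled while skydiving.
# Targeting reduces the hood physics weight.
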AnimBlueprint /BRCosmetics/Characters/Player/Male/Medium/Bodies/M_MED_Tiptoe/Meshes/Parts/M_MED_TipToe_AnimBP
target_skeleton=&/M_MED_TipToe_Skeleton
graphs [
 AnimGraph {nodes [
   n1 = Copy Pose From Mesh(SourceMeshComponent=Get Body_Mesh())
   n2 = Copy Pose From Mesh() {bCopyCurves=true, bUseAttachedParent=true}
   n3 = Get hoodFix_alpha_r() -> (hoodFix_alpha_r:double)
   n4 = "Transform (Modify) Bone - Bone: dyn_hood_base_r"(
    ComponentPose=Local To Component(LocalPose=Layered blend per bone(BasePose=@n2, BlendPoses_0=@n1) {layerSetup [{branchFilters [{boneName=upperarm_l}, {boneName=upperarm_r}]}]})
    Alpha=@n3
    Translation="0,0,0"
    Rotation="0,0,0"
    Scale="1,1,1"
   ) {
    rotationMode=BMM_Additive
    rotationSpace=BCS_BoneSpace
    translationMode=BMM_Additive
    translationSpace=BCS_BoneSpace
    alphaBoolBlend {blendOption=Linear}
    boneToModify {boneName=dyn_hood_base_r}
   }
   n5 = "Transform (Modify) Bone - Bone: dyn_hood_ro_base_r"(ComponentPose=@n4, Alpha=@n3, Translation="0,0,7", Rotation="0,0,90", Scale="1,1,1") {
    rotationMode=BMM_Additive
    rotationSpace=BCS_BoneSpace
    translationMode=BMM_Additive
    translationSpace=BCS_BoneSpace
    alphaBoolBlend {blendOption=Linear}
    boneToModify {boneName=dyn_hood_ro_base_r}
    rotation {roll=90}
    translation {z=7}
   }
   n6 = RigidBody(ComponentPose=@n5, Alpha=Get hoodDynAlpha() -> (hoodDynAlpha:double)) {
    bClampLinearTranslationLimitToRefPose=true
    bOverrideWorldGravity=true
    overridePhysicsAsset=&/M_MED_TipToe_Physics
    simulationSpace=BaseBoneSpace
    alphaBoolBlend {blendOption=Linear}
    baseBoneRef {boneName=spine_02}
    componentLinearAccScale=<0.2 0.2 0.2>
    componentLinearVelScale=<0.2 0.2 0.2>
    overrideWorldGravity {z=-980}
   }
   Output Pose(Result=Component To Local(ComponentPose=@n6)) {layerGroup=None}
  ]}
 EventGraph {nodes [
   Event BlueprintUpdateAnimation()
   > n1 = Sequence!(execute) -> (then_0, then_1, then_2:Exec, then_3:Exec)
   > n2 = Branch(Condition=Get bIsTargeting())
   > Set hoodDynAlpha(hoodDynAlpha:double=0.35) -> (Output_Get:double)
   Set hoodDynAlpha(execute=@n2.else, hoodDynAlpha=1)
   n3 = Branch(execute=@n1.then_2, Condition=Get bIsSkyDiving())
   > Set hoodFix_alpha_r(hoodFix_alpha_r:double=0) -> (Output_Get:double)
   n4 = Get armUp_R() -> (armUp_R:double)
   n5 = Get armForward_R() -> (armForward_R:double)
   Set hoodFix_alpha_r(
    execute=@n3.else
    hoodFix_alpha_r=float * float(A=MapRangeClamped(Value=@n4, InRangeA=-30, InRangeB=10), B=MapRangeClamped(Value=@n5, InRangeA=70, InRangeB=90))
   )
   n6 = GetOwnerSkeletalMeshComponent()
   n7 = BreakRotator(InRot=BreakTransform(InTransform=GetDeltaTransformFromRefPose(self=@n6, BoneName=upperarm_r, BaseName=clavicle_r)).Rotation)
   Set armForward_R(execute=@n1.then_1, armForward_R:double=@n7.Yaw) -> (Output_Get:double)
   > Set armUp_R(armUp_R:double=@n7.Pitch) -> (Output_Get:double)
   Event BlueprintInitializeAnimation()
   > Set Body_Mesh(Body_Mesh=GetOwnerSkeletalMeshComponent())
   > n8 = Cast To PlayerPawn_Athena(Object=TryGetPawnOwner())
   > n9 = Set Pawn Owner(Pawn Owner:Player Pawn Athena Object Reference=@n8) -> (Output_Get:Player Pawn Athena Object Reference)
   > Set Head_Mesh(Head_Mesh:Skeletal Mesh Component Object Reference=GetSkeletalMeshForPartType(self=@n9, PartType=Head)) -> (Output_Get:Skeletal Mesh Component Object Reference)
  ]}
]
variables [
 armForward_R {type=number}
 armUp_R {type=number}
 Body_Mesh {type=SkeletalMeshComponent}
 Head_Mesh {type=SkeletalMeshComponent}
 hoodDynAlpha {type=number, default=1}
 hoodFix_alpha_r {type=number}
 Pawn Owner {type=PlayerPawn_Athena_C}
]
